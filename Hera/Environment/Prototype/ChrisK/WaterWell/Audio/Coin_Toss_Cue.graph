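# Sound cue for a coin toss, filed with the prototype water well prop audio.
# A single wave player feeds the Coin_Toss_01 wave straight to the output, with no randomization or modulation.
# It uses medium prop attenuation and the Athena ambient props sound class, with a boosted volume multiplier.
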SoundCue /Hera/Environment/Prototype/ChrisK/WaterWell/Audio/Coin_Toss_Cue
root=WavePlayer_0
nodes [{id=WavePlayer_0, properties {soundWaveAssetPtr=&/Coin_Toss_01}}]
properties {
 attenuationSettings=&/ATT_Prop_Medium
 maxDistance=1500
 soundClassObject=&Fort_Ambient_Props_Athena
 volumeMultiplier=1.4
}
