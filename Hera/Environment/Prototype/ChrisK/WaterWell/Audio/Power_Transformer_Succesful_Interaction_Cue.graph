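# Sound cue for a successful interaction with a power transformer in the WaterWell prototype, playing the Generator_Fixed sound.
# A single wave player feeds the output directly, with no randomization or modulation.
# It uses the non-critical in-game UI sound class and overrides concurrency to allow one instance, stopping the oldest when retriggered.
# Its very large max distance suggests it is audible regardless of range.
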SoundCue /Hera/Environment/Prototype/ChrisK/WaterWell/Audio/Power_Transformer_Succesful_Interaction_Cue
root=WavePlayer_0
nodes [{id=WavePlayer_0, properties {soundWaveAssetPtr=&/Generator_Fixed}}]
properties {
 bOverrideConcurrency=true
 maxDistance=2097152
 soundClassObject=&Fort_UI_InGame_NonCritical
 volumeMultiplier=1
 concurrencyOverrides {maxCount=1, resolutionRule=StopOldest}
}
