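# Sound cue for a broken generator prop in the WaterWell prototype environment, named as a loop but built from one-shot waves.
# A Random node picks among three Broken_Generator_OneShots waves with equal weight and feeds an Attenuation node using the small-prop preset.
# The graph has no Looping or Delay node.
# It plays in the Athena ambient props sound class with a raised volume multiplier.
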SoundCue /Hera/Environment/Prototype/ChrisK/WaterWell/Audio/SC_Broken_Generator_Loop_Cue
root=Attenuation_0
nodes [
 {id=Attenuation_0, children [Random_0], properties {attenuationSettings=&/ATT_Prop_Small}}
 {id=Random_0, children [WavePlayer_1,WavePlayer_2,WavePlayer_3], properties {weights [1,1,1]}}
 {id=WavePlayer_1, properties {soundWaveAssetPtr=&/Broken_Generator_OneShots_07}}
 {id=WavePlayer_2, properties {soundWaveAssetPtr=&/Broken_Generator_OneShots_06}}
 {id=WavePlayer_3, properties {soundWaveAssetPtr=&/Broken_Generator_OneShots_05}}
]
properties {maxDistance=1250, soundClassObject=&Fort_Ambient_Props_Athena, volumeMultiplier=1.3}
