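# Sound cue for the prototype water well fountain, playing ambient running water on the Fort_Ambient_Athena class.
# A mixer layers a looping medium stream wave with a non-looping Fountain_Fill wave mixed at double its volume.
# The stream branch passes through a modulator for pitch variation and a Looping node.
# It uses the ATT_Prop_Medium attenuation asset and overrides concurrency to allow one instance at a time.
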SoundCue /Hera/Environment/Prototype/ChrisK/WaterWell/Audio/SC_Fountain_Water_Loop
root=Mixer_1
nodes [
 {id=Looping_0, children [Modulator_1]}
 {id=Mixer_1, children [Looping_0,WavePlayer_2], properties {inputVolume [1,2]}}
 {id=Modulator_1, children [WavePlayer_1], properties {pitchMax=1, volumeMax=1, volumeMin=1}}
 {id=WavePlayer_1, properties {bLooping=true, soundWaveAssetPtr=&/stream_medium_A_03}}
 {id=WavePlayer_2, properties {soundWaveAssetPtr=&/Fountain_Fill}}
]
properties {
 attenuationSettings=&/ATT_Prop_Medium
 bOverrideConcurrency=true
 maxDistance=1500
 soundClassObject=&Fort_Ambient_Athena
 volumeMultiplier=1.3
 attenuationOverrides {attenuationShape=Capsule, falloffDistance=950, stereoSpread=0, attenuationShapeExtents=<250 90>}
 concurrencyOverrides {maxCount=1}
}
