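# Sound cue for a repaired generator in the water well prototype, pairing a fixed-generator sound with a running loop.
# A mixer plays the Generator_Fixed wave once at full input, alongside the Broken_Generator_Loop wave at half volume.
# The loop branch starts after a one-second delay, passing through looping and volume-envelope nodes with its own medium-prop attenuation.
# The cue also uses medium-prop attenuation and limits concurrency to one instance.
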SoundCue /Hera/Environment/Prototype/ChrisK/WaterWell/Audio/SC_Generator_Fixed_Cue
root=Mixer_1
nodes [
 {id=Attenuation_1, children [Delay_1], properties {attenuationSettings=&/ATT_Prop_Medium}}
 {id=Delay_1, children [Looping_1], properties {delayMax=1, delayMin=1}}
 {
  id=Enveloper_0
  children [WavePlayer_1]
  properties {
   bLoopIndefinitely=true
   loopEnd=1
   loopStart=1
   pitchCurve {editorCurveData {keys [{time=0}]}}
   volumeCurve {editorCurveData {keys [<at 0 0>,<at 0.0019 0.5132>,<at 0.003 0.5099>,<at 0.004 0>,<at 1 1>]}}
  }
 }
 {id=Looping_1, children [Enveloper_0]}
 {id=Mixer_1, children [Attenuation_1, WavePlayer_0], properties {inputVolume [0.5,1]}}
 {id=WavePlayer_0, properties {soundWaveAssetPtr=&/Generator_Fixed}}
 {id=WavePlayer_1, properties {bLooping=true, soundWaveAssetPtr=&/Broken_Generator_Loop}}
]
properties {
 attenuationSettings=&ATT_Prop_Medium
 bOverrideConcurrency=true
 maxDistance=1500
 volumeMultiplier=1.5
 concurrencyOverrides {maxCount=1}
}
