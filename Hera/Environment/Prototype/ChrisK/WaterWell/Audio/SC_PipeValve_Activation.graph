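# Sound cue that plays the activation sound for a pipe valve prop in the water well prototype environment.
# A single wave player feeds the PipeValve_Activation wave straight to the output, routed through the Fort_Ambient_Athena sound class with a slight volume boost.
# It references the small-prop attenuation asset, carries capsule-shaped attenuation override values with no stereo spread, and overrides concurrency.
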SoundCue /Hera/Environment/Prototype/ChrisK/WaterWell/Audio/SC_PipeValve_Activation
root=WavePlayer_0
nodes [{id=WavePlayer_0, properties {soundWaveAssetPtr=&/PipeValve_Activation}}]
properties {
 attenuationSettings=&/ATT_Prop_Small
 bOverrideConcurrency=true
 maxDistance=1250
 soundClassObject=&Fort_Ambient_Athena
 volumeMultiplier=1.3
 attenuationOverrides {attenuationShape=Capsule, falloffDistance=950, stereoSpread=0, attenuationShapeExtents=<250 90>}
}
