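# Sound cue for a water well pipe valve: water starts through the pipe, sprays out and settles into a running stream loop.
# A mixer layers a pipe start one-shot, a delayed spray start one-shot at reduced volume, and the looping stream.
# The stream is briefly delayed, pitched up by a fixed modulator and faded in by a volume envelope under a looping node.
# It plays as ambient sound with small-prop attenuation.
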SoundCue /Hera/Environment/Prototype/ChrisK/WaterWell/Audio/SC_PipeValve_Water_Loop
root=Mixer_0
nodes [
 {id=Delay_0, children [Enveloper_0], properties {delayMax=0.4, delayMin=0.4}}
 {id=Delay_1, children [WavePlayer_0], properties {delayMax=0.928, delayMin=0.928}}
 {
  id=Enveloper_0
  children [WavePlayer_1]
  properties {
   pitchCurve {editorCurveData {keys [{time=0}]}}
   volumeCurve {editorCurveData {keys [
      <at 0 0.0113 RCIM_Cubic>
      {arriveTangent=0.6732, interpMode=RCIM_Cubic, leaveTangent=0.6732, tangentMode=RCTM_User, time=0.9285, value=0.3172}
      <at 1.572 0.9931>
     ]}}
  }
 }
 {id=Looping_0, children [Modulator_1]}
 {id=Mixer_0, children [Looping_0,Modulator_0,Delay_1], properties {inputVolume [1,1.4,0.5]}}
 {id=Modulator_0, children [WavePlayer_2], properties {pitchMax=1, pitchMin=1}}
 {id=Modulator_1, children [Delay_0], properties {pitchMax=2, pitchMin=2, volumeMax=1, volumeMin=1}}
 {id=WavePlayer_0, properties {soundWaveAssetPtr=&/Fountain_Water_Spray_Start_02}}
 {id=WavePlayer_1, properties {bLooping=true, soundWaveAssetPtr=&/stream_medium_A_03}}
 {id=WavePlayer_2, properties {soundWaveAssetPtr=&/Pipe_water_startB}}
]
properties {
 attenuationSettings=&/ATT_Prop_Small
 bOverrideConcurrency=true
 maxDistance=1250
 soundClassObject=&Fort_Ambient_Athena
 volumeMultiplier=1
 attenuationOverrides {attenuationShape=Capsule, falloffDistance=950, stereoSpread=0, attenuationShapeExtents=<250 90>}
}
