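; Interactable power switch for the water well fountain that lets a player restore power to it once.
; On BeginPlay a scalable float toggle either resets the valve state or disables interaction.
; The first interaction locks further use, calls PowerRestored on the fountain manager, triggers non-authority-only activation audio and Niagara effects, and sets PowerRestored.
; The PowerRestored RepNotify removes the looping sparks and fades out the electricity hum.
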(event EventBeginPlay
  ; Init
  (if (ScalableFloat|GetValueasBool (Variables|Default|GetEnabledHF))
    (Variables|Default|SetValveActive false)
    (else
      (Variables|Interaction|SetbAllowInteract false))))

(event Interaction|EventBlueprintOnInteract
  (params (InteractingPawn :type Fort%20Pawn%20Object%20Reference) (InteractionBeingAttempted :type EInteractionBeingAttempted%20Enum :default "FirstInteraction"))
  ; Turn the valve
  (Utilities|FlowControl|DoOnce false
    (:Completed
      (Variables|Default|SetCanPlayerInteract false)
      (Class|BPFountainManager|PowerRestored (Variables|Default|GetFountainManagerRef))
      (CallFunction|ClientFX)
      (Variables|Default|SetPowerRestored true))))

(event Custom|ClientFX
  ; Client sounds and FX
  (Networking|SwitchHasAuthority
    (:Remote
      (Audio|Components|Audio|Play (Variables|Components|GetAudio_Activated) 0.0)
      (Niagara|SpawnSystemAttached "/Game/Environments/Hera/FX/Props/Fountain/NS_Fountain_ElectricalSwitch.NS_Fountain_ElectricalSwitch" (Variables|Components|GetDefaultSceneRoot) "None" "14.250000,0.000000,170.200000"))))

(fn BlueprintCanInteract
  (params (InteractingPawn :type Fort%20Pawn%20Object%20Reference) (InteractionBeingAttempted :type EInteractionBeingAttempted%20Enum :default "FirstInteraction") (InteractionType :type TInteractionType%20Enum :default "IT_NoInteraction"))
  (return (Variables|Default|GetCanPlayerInteract)))

(fn BlueprintGetInteractionString
  (params (InteractingPawn :type Fort%20Pawn%20Object%20Reference) (InteractionBeingAttempted :type EInteractionBeingAttempted%20Enum :default "FirstInteraction"))
  (return (Variables|Default|GetInteractionString)))

(fn GetBoundsComponents
  (params (SelfActor :type Actor%20Object%20Reference))
  (return (Utilities|Array|MakeArray (Variables|Components|GetSwitchMesh))))

(fn OnRep_PowerRestored
  (if (Variables|Default|GetPowerRestored)
    (Components|DestroyComponent (Variables|Default|GetNS_Fountain_ElectricalSwitch_SparksLooping))
    (Audio|Components|Audio|FadeOut (Variables|Default|GetElectricityLoop) 0.5 0.0 "Linear" :!class /Script/Engine.AudioComponent)))

(fn ConstructionScript)
