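; Interactive fountain pipe valve that players turn once, after EnablePipeValve sets power restored and allows interaction.
; BeginPlay caches start and end rotations and disables interaction if the EnabledHF scalable float is false.
; Interacting marks the water tower repaired, runs remote-only sound and Niagara, and rotates the mesh by timeline before notifying BP_FountainManager.
; RepNotifies start looping water FX and audio on power restore and stop them on repair.
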(event EventBeginPlay
  ; Init
  (bind _output_get (Variables|Default|SetInitialRelativeRotation (Variables|/Script/Engine.SceneComponent|GetRelativeRotation (Variables|Components|GetValveMesh))))
  (Variables|Default|SetEndRelativeRotation (Math|Rotator|CombineRotators _output_get "-179.000000,0.000000,0.000000"))
  (if (not (ScalableFloat|GetValueasBool (Variables|Default|GetEnabledHF)))
    (Variables|Interaction|SetbAllowInteract false)))

(event Custom|EnablePipeValve
  ; Emable Valve Interaction
  (Variables|Default|SetCanPlayerInteract true)
  (Variables|Default|SetPowerRestored true))

(event Interaction|EventBlueprintOnInteract
  (params (InteractingPawn :type Fort%20Pawn%20Object%20Reference) (InteractionBeingAttempted :type EInteractionBeingAttempted%20Enum :default "FirstInteraction"))
  ; Turn the valve
  (Utilities|FlowControl|DoOnce false
    (:Completed
      (Variables|Default|SetCanPlayerInteract false)
      (Variables|Default|SetWaterTowerRepaired true)
      (CallFunction|ClientFX)
      (bind (_direction _rotatevalve) (Timeline|Turn PlayFromStart)
        (:Update
          (Transformation|SetRelativeRotation (Variables|Components|GetValveMesh) (Math|Rotator|Lerp(Rotator) (Variables|Default|GetInitialRelativeRotation) (Variables|Default|GetEndRelativeRotation) _rotatevalve true) false false))
        (:Finished
          (Class|BPFountainManager|ValveTurned (Variables|Default|GetFountainManagerRef)))))))

(event Custom|ClientFX
  ; Client sounds and FX
  (Networking|SwitchHasAuthority
    (:Remote
      (Audio|Components|Audio|Play (Variables|Components|GetAudio_TurnActivatedForward) 0.0)
      (Niagara|SpawnSystemAttached "/Game/Environments/Hera/FX/Props/Fountain/NS_Fountain_WaterPipe.NS_Fountain_WaterPipe" (Variables|Components|GetDefaultSceneRoot) "None" "0.000000,63.000000,-12.000000"))))

(fn BlueprintCanInteract
  (params (InteractingPawn :type Fort%20Pawn%20Object%20Reference) (InteractionBeingAttempted :type EInteractionBeingAttempted%20Enum :default "FirstInteraction") (InteractionType :type TInteractionType%20Enum :default "IT_NoInteraction"))
  (return (Variables|Default|GetCanPlayerInteract)))

(fn BlueprintGetInteractionString
  (params (InteractingPawn :type Fort%20Pawn%20Object%20Reference) (InteractionBeingAttempted :type EInteractionBeingAttempted%20Enum :default "FirstInteraction"))
  (return (Variables|Default|GetInteractionString)))

(fn GetBoundsComponents
  (params (SelfActor :type Actor%20Object%20Reference))
  (return (Utilities|Array|MakeArray (Variables|Components|GetValveMesh))))

(fn OnRep_PowerRestored
  (if (Variables|Default|GetPowerRestored)
    (Niagara|ReinitializeSystem (Variables|Default|GetNS_Fountain_WaterLooping))
    (Audio|Components|Audio|Play (Variables|Default|GetWaterLoop) 0.0)))

(fn OnRep_WaterTowerRepaired
  (if (Variables|Default|GetWaterTowerRepaired)
    (Components|Activation|Deactivate (Variables|Default|GetNS_Fountain_WaterLooping))
    (Audio|Components|Audio|FadeOut (Variables|Default|GetWaterLoop) 2.0 0.0 "Linear" :!class /Script/Engine.AudioComponent)))

(fn ConstructionScript)
