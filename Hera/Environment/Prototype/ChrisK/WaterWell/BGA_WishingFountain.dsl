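; Interactable wishing fountain that takes gold and tosses out tiered loot.
; StartWater runs once, filling the basin by timeline; replicated WaterActive drives water visuals, loop audio and interactability.
; Each controller wishes once, gated by gold, cooldown and total-wish limit; a wish removes gold, plays cosmetics, then after a delay spawns pickups thrown along indexed spread vectors with random strength.
; Hotfixable scalable floats control enablement, cost, uses and timing.
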(event Custom|DebugPrintWishingFountainHotFixes
  ; Debug Commands
  (Development|PrintString (Utilities|String|ToString(Text) (Utilities|Text|FormatText "NSLOCTEXT(\"[0157B837EB90D710A0A282ED9FBE9D9D]\", \"0BDC3D494D6B08F02FD66EA6ADEC0FF9\", \"WISHING FOUNTAIN - Enabled {enabled}, NumUses {numuses}, GoldCost {goldcost}\")" (ScalableFloat|GetValueasBool (Variables|Hotfixable|GetFountainEnabled)) (ScalableFloat|GetValueasInteger (Variables|Hotfixable|GetFountainNumOfAllowedUses)) (ScalableFloat|GetValueasInteger (Variables|Hotfixable|GetFountainGoldCostPerUse)))) true true "(R=1.000000,G=0.070857,B=0.000000,A=1.000000)" 1e+01))

(event Custom|PlayMusicStinger
  (params (Player%20Controller :type Fort%20Player%20Controller%20Athena%20Object%20Reference))
  ; Music Stinger
  (Networking|CanExecuteCosmeticEvents
    (:True
      (if (Pawn|IsLocalPlayerController Player%20Controller)
        (Audio|PlaySound2D "/Hera/Environment/Prototype/ChrisK/WaterWell/Audio/Coin_Toss_Cue.Coin_Toss_Cue")
        (Utilities|FlowControl|Delay 1.5)
        (Audio|AddEvent (GameInstanceSubsystems|GetMusicEventSubsystem) "/Hera/Audio/Music/Gameplay/SC_S41_WishingFountain_STNG.SC_S41_WishingFountain_STNG" "(TagName=\"MusicEventSubsystem.Event.Gameplay\")" "(GameplayTags=)")))))

(event EventBeginPlay
  (bind _fountainwatermesh (Variables|Default|GetFountainWaterMesh))
  (bind _returnvalue (Transformation|GetWorldLocation _fountainwatermesh))
  ; Init
  (if (ScalableFloat|GetValueasBool (Variables|Hotfixable|GetFountainEnabled))
    (CallFunction|InitHotFixValues)
    (Networking|SwitchHasAuthority
      (:Authority
        (Variables|BuildingActor|SetbIsInvulnerable true)
        (Variables|Default|SetWaterActive false)
        (Variables|Default|SetWaterEndLocation _returnvalue)
        (Variables|Default|SetWaterStartLocation (- _returnvalue "0.000000,0.000000,32.000000"))))
    (else
      (Variables|Interaction|SetbAllowInteract false))))

(event Custom|StartWater
  ; Start Water
  ;
  (Utilities|FlowControl|DoOnce false
    (:Completed
      (Networking|SwitchHasAuthority
        (:Authority
          (Utilities|FlowControl|Sequence
            (:then_0
              (bind (_direction _water) (Timeline|FillWater PlayFromStart)
                (:Update
                  (Transformation|SetWorldLocation (Variables|Default|GetFountainWaterMesh) (Math|Vector|Lerp(Vector) (Variables|Default|GetWaterStartLocation) (Variables|Default|GetWaterEndLocation) _water) false false)
                  (Rendering|Material|SetScalarParameterValueonMaterials (Variables|Default|GetSM_Fountain_WaterSplines_01) "Fountain_Water_Progress" _water)
                  (Niagara|SetNiagaraVariable(Float) (Variables|Default|GetNS_Outdoor_FountainBase) "Fill" _water))))
            (:then_1
              (Utilities|FlowControl|Delay 0.1)
              (Variables|Default|SetWaterActive true))))
        (:Remote
          (Audio|Components|Audio|Play (Variables|Default|GetAudio_WaterFountainStart) 0.0)
          (Effects|Components|ParticleSystem|SpawnFXAtLocation "/Game/Environments/Hera/FX/Props/Fountain/NS_Fountain_Activated.NS_Fountain_Activated" (+ (Transformation|GetWorldLocation (Variables|Default|GetScene)) "0.000000,0.000000,452.000000"))
          (Audio|Components|Audio|Play (Variables|Default|GetFountainWishAudio) 0.0))))))

(event Interaction|EventBlueprintOnInteract
  (params (InteractingPawn :type Fort%20Pawn%20Object%20Reference) (InteractionBeingAttempted :type EInteractionBeingAttempted%20Enum :default "FirstInteraction"))
  ; Make a Wish
  (Variables|Default|SetCurrentWishingPlayerRef InteractingPawn)
  (bind _asfort_player_controller_athena (Utilities|Casting|CastToFortPlayerControllerAthena (Pawn|GetController InteractingPawn)))
  (Utilities|Array|Add (Variables|Default|GetPlayersWhoHaveWished) _asfort_player_controller_athena)
  (Utilities|FlowControl|Sequence
    (:then_0
      (CallFunction|RemoveGold (Variables|Default|GetCurrentWishingPlayerRef))
      (CallFunction|WishMadeCosmetics)
      (Math|Integer|IncrementInt (Variables|Default|GetCurrentNumWishesGranted)
        (:%20%20
          (Utilities|FlowControl|Delay (ScalableFloat|GetValueatLevel (Variables|Hotfixable|GetDelayFromInteractToSpawnLoot)))
          (CallFunction|GiveRewards)
          (Variables|Default|SetNextWishCooldownActive true))))
    (:then_1
      (Utilities|FlowControl|Delay (ScalableFloat|GetValueatLevel (Variables|Hotfixable|GetCooldownBetweenUses)))
      (Variables|Default|SetNextWishCooldownActive false)
      (if (>= (Variables|Default|GetCurrentNumWishesGranted) (Variables|Default|GetNumOfAllowedWishes))
        (Variables|Default|SetMaxWishesReached true)
        (Development|PrintString "Max Wishes Granted This Session" true true "(R=0.000000,G=0.660000,B=1.000000,A=1.000000)" 5.0)))
    (:then_2
      (CallFunction|PlayMusicStinger _asfort_player_controller_athena))))

(event Custom|WishMadeCosmetics
  (bind _scene (Variables|Default|GetScene))
  (bind _returnvalue (Transformation|GetWorldLocation _scene))
  ; Wish Made Cosmetics
  (Networking|SwitchHasAuthority
    (:Remote
      (Effects|Components|ParticleSystem|SpawnFXAtLocation "/Game/Environments/Hera/FX/Props/NS_Fountain_Wish.NS_Fountain_Wish" _returnvalue)
      (Audio|Components|Audio|Play (Variables|Default|GetFountainWishAudio) 0.0)
      (Utilities|FlowControl|Delay (ScalableFloat|GetValueatLevel (Variables|Hotfixable|GetDelayFromInteractToSpawnLoot)))
      (Audio|PlaySoundatLocation "/Game/Athena/Sounds/Quests/Athena_CollectHiddenStarReward_Cue.Athena_CollectHiddenStarReward_Cue" _returnvalue "0, 0, 0" 1.0 1.0 0.0 "/Game/Athena/Sounds/Events/Papaya/SpeakerSystem/Papaya_Speaker_Medium_Att.Papaya_Speaker_Medium_Att")
      (Effects|Components|ParticleSystem|SpawnFXAtLocation "/Game/Environments/Hera/FX/Props/Fountain/NS_Fountain_Wish_Spawn.NS_Fountain_Wish_Spawn" (Transformation|GetWorldLocation (Variables|Default|GetArrow))))))

(event Custom|GiveRewards
  (bind _arrow (Variables|Default|GetArrow))
  (bind _returnvalue (Transformation|GetForwardVector _arrow))
  (bind _tossvelocity (Variables|Default|GetTossVelocity))
  (bind _returnvalue_2 (Math|Conversions|ToVector(Float) _tossvelocity))
  (bind _returnvalue_3 (* _returnvalue _returnvalue_2))
  ; Give Loot
  (Utilities|FlowControl|DoN (Variables|Default|GetNumOfAllowedWishes)
    (:Exit
      (Variables|Default|SetImpulse _returnvalue_3)
      (bind _returnvalue_4 (Inventory|SpawnPickupInWorldWithLootTier (Variables|Default|GetLootTierToFire) (Transformation|GetWorldLocation _arrow) 0 0 true true _returnvalue_3))
      (Variables|Default|SetLoot _returnvalue_4)
      (for (_array_element _array_index) (Variables|Default|GetLoot)
        (bind _asfort_pickup_athena (Utilities|Casting|CastToFortPickupAthena _array_element))
        (bind _yeetvector (CallFunction|GetYeetVectors _array_index))
        (Pickup|TossPickupinDirection _asfort_pickup_athena (+ (* _yeetvector (Math|Conversions|ToVector(Float) (Math|Random|RandomFloatinRange 0.8 1.05))) (* (Transformation|GetForwardVector (Variables|Default|GetArrow)) (* (Variables|Default|GetTossVelocity) (Math|Random|RandomFloatinRange 0.85 1.05)))) self false false)))))

(fn BlueprintCanInteract
  (params (InteractingPawn :type Fort%20Pawn%20Object%20Reference) (InteractionBeingAttempted :type EInteractionBeingAttempted%20Enum :default "FirstInteraction") (InteractionType :type TInteractionType%20Enum :default "IT_NoInteraction"))
  (bind _asfort_player_controller_athena (Utilities|Casting|CastToFortPlayerControllerAthena (Pawn|GetController InteractingPawn)))
  (if (Utilities|Array|ContainsItem (Variables|Default|GetPlayersWhoHaveWished) _asfort_player_controller_athena)
    (return false)
    (elif (Variables|Default|GetWaterActive)
      (if (CallFunction|HasEnoughGold (Variables|GetInteractingPawn))
        (if (Variables|Default|GetMaxWishesReached)
          _
          (elif (Variables|Default|GetNextWishCooldownActive)
            _
            (else
              (return true)))))))
  (return false))

(fn BlueprintGetFailedInteractionString
  (params (InteractingPawn :type Fort%20Pawn%20Object%20Reference) (InteractionBeingAttempted :type EInteractionBeingAttempted%20Enum :default "FirstInteraction"))
  (if (Variables|Default|GetWaterActive)
    (bind _asfort_player_controller_athena (Utilities|Casting|CastToFortPlayerControllerAthena (Pawn|GetController InteractingPawn)))
    (if (not (Utilities|Array|ContainsItem (Variables|Default|GetPlayersWhoHaveWished) _asfort_player_controller_athena))
      (if (Variables|Default|GetMaxWishesReached)
        (return (Variables|Default|GetNoMoreWishesText))
        (else
          (return (CallFunction|GetInteractionStringwithCost))))
      (else
        _)))
  (return ""))

(fn BlueprintGetInteractionString
  (params (InteractingPawn :type Fort%20Pawn%20Object%20Reference) (InteractionBeingAttempted :type EInteractionBeingAttempted%20Enum :default "FirstInteraction"))
  (return (CallFunction|GetInteractionStringwithCost)))

(fn BlueprintGetInteractionTime
  (params (InteractingPawn :type Fort%20Pawn%20Object%20Reference) (InteractionBeingAttempted :type EInteractionBeingAttempted%20Enum :default "FirstInteraction"))
  (return 1.25 true))

(fn GetBoundsComponents
  (params (SelfActor :type Actor%20Object%20Reference))
  (return (Utilities|Array|MakeArray (Variables|Default|GetSM_Outdoor_FountainBaseLg_B))))

(fn GetInteractionStringWithCost
  (return (Utilities|Text|FormatText "NSLOCTEXT(\"[0157B837EB90D710A0A282ED9FBE9D9D]\", \"E4FA9A224E49584D3E694EB9B935D0B3\", \"{Interaction} ({Cost} {Gold})\")" (Variables|Default|GetFountainInteractionText) (Variables|Default|GetGoldCostPerWish) (Variables|Default|GetGoldInteractionString))))

(fn GetYeetVectors
  (params (Index :type Integer))
  (if (> Index 11)
    (LocalVariables|SetLocalIndex 0)
    (else
      (LocalVariables|SetLocalIndex Index)))
  (switch int (LocalVariables|GetLocalIndex)
    (:0
      (LocalVariables|SetLocalYeetVector (Math|Vector|MakeVector (* (Variables|Default|GetTossSpreadMultiplier) -3.5e+02) (* 0 (Variables|Default|GetTossSpreadMultiplier)) (* 5e+02 (Variables|Default|GetTossSpreadMultiplier))) :!label _label))
    (:1
      (LocalVariables|SetLocalYeetVector (Math|Vector|MakeVector (* (Variables|Default|GetTossSpreadMultiplier) 3.5e+02) (* 0 (Variables|Default|GetTossSpreadMultiplier)) (* 5e+02 (Variables|Default|GetTossSpreadMultiplier))) :!label _label_2))
    (:2
      (LocalVariables|SetLocalYeetVector (Math|Vector|MakeVector (* (Variables|Default|GetTossSpreadMultiplier) 0) (* 3.5e+02 (Variables|Default|GetTossSpreadMultiplier)) (* 5e+02 (Variables|Default|GetTossSpreadMultiplier))) :!label _label_3))
    (:3
      (LocalVariables|SetLocalYeetVector (Math|Vector|MakeVector (* (Variables|Default|GetTossSpreadMultiplier) 0) (* 3.5e+02 (Variables|Default|GetTossSpreadMultiplier)) (* 5e+02 (Variables|Default|GetTossSpreadMultiplier))) :!label _label_4))
    (:4
      (LocalVariables|SetLocalYeetVector (Math|Vector|MakeVector (* (Variables|Default|GetTossSpreadMultiplier) 3.5e+02) (* 3.5e+02 (Variables|Default|GetTossSpreadMultiplier)) (* 5e+02 (Variables|Default|GetTossSpreadMultiplier))) :!label _label_5))
    (:5
      (LocalVariables|SetLocalYeetVector (Math|Vector|MakeVector (* (Variables|Default|GetTossSpreadMultiplier) -3.5e+02) (* -3.5e+02 (Variables|Default|GetTossSpreadMultiplier)) (* 5e+02 (Variables|Default|GetTossSpreadMultiplier))) :!label _label_6))
    (:6
      (goto _label))
    (:7
      (goto _label_2))
    (:8
      (goto _label_3))
    (:9
      (goto _label_4))
    (:10
      (goto _label_5))
    (:11
      (goto _label_6)))
  (return (LocalVariables|GetLocalYeetVector)))

(fn HasEnoughGold
  (params (PlayerPawn :type Fort%20Pawn%20Object%20Reference))
  (bind _returnvalue (Pawn|GetController PlayerPawn))
  (Utilities|IsValid PlayerPawn
    (:Is%20Valid
      (Utilities|IsValid _returnvalue
        (:Is%20Valid
          (bind _asfort_player_controller_athena (Utilities|Casting|CastToFortPlayerControllerAthena _returnvalue)
            (:then
              (bind _returnvalue_2 (InventoryManagement|GetItemCountforItemDefinition _asfort_player_controller_athena (Variables|Default|GetGoldItemDefinition)))
              (return (>= _returnvalue_2 (Variables|Default|GetGoldCostPerWish))))
            (:CastFailed
              _)))
        (:Is%20Not%20Valid
          _)))
    (:Is%20Not%20Valid
      _))
  (return false))

(fn InitHotFixValues
  (Variables|Default|SetNumOfAllowedWishes (ScalableFloat|GetValueasInteger (Variables|Hotfixable|GetFountainNumOfAllowedUses)))
  (Variables|Default|SetGoldCostPerWish (ScalableFloat|GetValueasInteger (Variables|Hotfixable|GetFountainGoldCostPerUse))))

(fn OnRep_WaterActive
  (if (Variables|Default|GetWaterActive)
    (Rendering|SetVisibility (Variables|Default|GetFountainWaterMesh) true false)
    (Components|Activation|Activate (Variables|Default|GetNS_Outdoor_FountainBase) true)
    (Audio|Components|Audio|Play (Variables|Default|GetAudio_WaterFountainLoop) 0.0)
    (else
      (Rendering|SetVisibility (Variables|Default|GetFountainWaterMesh) false false)
      (Components|Activation|Deactivate (Variables|Default|GetNS_Outdoor_FountainBase))
      (Audio|Components|Audio|FadeOut (Variables|Default|GetAudio_WaterFountainLoop) 1.0 0.0 "Linear" :!class /Script/Engine.AudioComponent))))

(fn RemoveGold
  (params (InteractingPlayer :type Fort%20Pawn%20Object%20Reference))
  (bind _returnvalue (Pawn|GetController InteractingPlayer))
  (Networking|SwitchHasAuthority
    (:Authority
      (Utilities|IsValid InteractingPlayer
        (:Is%20Valid
          (Utilities|IsValid _returnvalue
            (:Is%20Valid
              (bind _asfort_player_controller_athena (Utilities|Casting|CastToFortPlayerControllerAthena _returnvalue))
              (bind _returnvalue_2 (InventoryManagement|FindExistingItemforDefinition _asfort_player_controller_athena (Variables|Default|GetGoldItemDefinition)))
              (Utilities|IsValid _returnvalue_2
                (:Is%20Valid
                  (Inventory|RemoveFortItemFromPlayer _asfort_player_controller_athena _returnvalue_2 (Variables|Default|GetGoldCostPerWish)))))))))))

(fn ConstructionScript
  (if (Variables|Default|GetWaterActive)
    (Rendering|SetVisibility (Variables|Default|GetFountainWaterMesh) true false)
    (Rendering|SetVisibility (Variables|Default|GetNS_Outdoor_FountainBase) true false)
    (Audio|Components|Audio|Play (Variables|Default|GetAudio_WaterFountainLoop) 0.0)
    (else
      (Rendering|SetVisibility (Variables|Default|GetFountainWaterMesh) false false)
      (Rendering|SetVisibility (Variables|Default|GetNS_Outdoor_FountainBase) false false)
      (Audio|Components|Audio|Stop (Variables|Default|GetAudio_WaterFountainLoop)))))
