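; BP_Fountain_Manager sequences a prototype water-well fountain puzzle linking a pipe valve, wishing fountain and hydrants.
; PowerRestored enables the valve on authority, with only a debug print on clients.
; ValveTurned starts the fountain water, then after a delay ActivateHydrants caches hydrant locations, calls SpawnClientFX, and destroys the hydrants shortly after.
; Each stage is DoOnce-guarded, and SpawnClientFX plays Niagara effects and sounds at cached locations only on non-authority instances.
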(event EventBeginPlay
  ; Init
  (Networking|SwitchHasAuthority))

(event Custom|PowerRestored
  ; Power Restored
  (Utilities|FlowControl|DoOnce false
    (:Completed
      (Networking|SwitchHasAuthority
        (:Authority
          (Class|BGAPipeValveInteractiveFountain|EnablePipeValve (Variables|Default|GetPIpeValveRef)))
        (:Remote
          (Development|PrintString "Client FX"))))))

(event Custom|ValveTurned
  ; Valve Turned
  (Utilities|FlowControl|DoOnce false
    (:Completed
      (Class|BGAWishingFountain|StartWater (Variables|Default|GetFountainRef))
      (Utilities|FlowControl|Delay 2.5)
      (CallFunction|ActivateHydrants))))

(event Custom|ActivateHydrants
  ; Hydrants
  (Utilities|FlowControl|DoOnce false
    (:Completed
      (for _array_element (Variables|Default|GetHydrants)
        (Utilities|IsValid _array_element
          (:Is%20Valid
            (Utilities|Array|AddUnique (Variables|Default|GetAllHydrantLocations) (Transformation|GetActorLocation _array_element)))))
      (CallFunction|SpawnClientFX)
      (Utilities|FlowControl|Delay 0.25)
      (for _array_element_2 (Variables|Default|GetHydrants)
        (Utilities|IsValid _array_element_2
          (:Is%20Valid
            (Actor|DestroyActor _array_element_2)))))))

(event Custom|SpawnClientFX
  ; Hydrants
  (Networking|SwitchHasAuthority
    (:Remote
      (for _array_element (Variables|Default|GetAllHydrantLocations)
        (Niagara|SpawnSystematLocation "/Game/Athena/Items/QuestInteractables/Generic/FX/NS_Athena_QuestItem_PickupSuccess.NS_Athena_QuestItem_PickupSuccess" _array_element)
        (Audio|SpawnSoundatLocation "/Game/Sounds/Fort_Destructibles/Propane_Tank/Destructible_PropaneTank_Pre_Explode_Cue.Destructible_PropaneTank_Pre_Explode_Cue" _array_element)))))

(fn ConstructionScript)
